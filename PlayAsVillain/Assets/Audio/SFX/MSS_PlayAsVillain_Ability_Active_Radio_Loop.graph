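# Mono MetaSound source for the PlayAsVillain active-ability radio loop, spatialized by its own attenuation asset.
# OnPlay starts a Trigger Delay, which then plays the wave asset and sets a gain target to full.
# An InterpTo smooths that target over Fade.FadeTime and multiplies the audio, while Fade.FadeOut resets the target to fade out.
# The Fade.FadeIn input is unconnected, so the fade-in follows only the delayed start.
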MetaSoundSource /PlayAsVillain/Assets/Audio/SFX/MSS_PlayAsVillain_Ability_Active_Radio_Loop
inputs [
 Fade.FadeIn {type=Trigger}
 Fade.FadeOut {type=Trigger}
 Fade.FadeTime {type=Time}
 UE.Source.OnPlay {type=Trigger}
 Wave Asset {type=WaveAsset}
]
interfaces [UE.OutputFormat.Mono:1.0, UE.Source:1.1]
outputs ["UE.OutputFormat.Mono.Audio:0" {type=Audio}]
pages [Default {
  input_defaults {
   Fade.FadeIn=null
   Fade.FadeOut=null
   Fade.FadeTime=0.168
   UE.Source.OnPlay=null
   Wave Asset=&/PlayAsVillain_Ability_Active_Radio_Loop_03
  }
  nodes [
   TriggerDelay_0 = UE.Trigger Delay(In=^UE.Source.OnPlay) {also {defaults {Delay Time=4}}}
   WavePlayer_0 = UE.Wave Player:Mono(Wave Asset=^Wave Asset, Play=@TriggerDelay_0.Out) {also {defaults {Loop="<MetaSoundBuilderResult.SUCCEEDED: 0>"}, version [1,1]}}
   Multiply_0 = UE.Multiply:Audio by Float(
    AdditionalOperands=UE.InterpTo:Audio(
     Interp Time=^Fade.FadeTime
     Target=Value.Value:Float(Reset=^Fade.FadeOut, Set=@TriggerDelay_0.Out) {also {defaults {Target Value=1}}}.Output Value
    ).Value
    PrimaryOperand=@WavePlayer_0.Out Mono
   )
  ]
  links [{dst="@UE.OutputFormat.Mono.Audio:0", src=Multiply_0.Out}]
 }]
properties {
 assetImportData=&self:AssetImportData
 attenuationSettings=&/ATT_PlayAsVillain_Ability_Active_Radio
 bLooping=true
 duration=10000
 maxDistance=8500
 sampleRate=48000
 soundClassObject=&Fort_Gadgets_Items
 volume=0.3
 concurrencySet [&CS_Fort_Item]
}
