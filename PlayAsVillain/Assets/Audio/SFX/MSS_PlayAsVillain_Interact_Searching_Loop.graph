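# Stereo MetaSound source for the looping searching interaction sound in PlayAsVillain, spatialized with item interact attenuation.
# On play, it starts the InteractLoop wave and sets a stored float from its low initial value to full.
# An InterpTo smooths that value over Interp Time and multiplies both channels as gain, giving a volume ramp.
# The Reset trigger input resets the stored value.
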MetaSoundSource /PlayAsVillain/Assets/Audio/SFX/MSS_PlayAsVillain_Interact_Searching_Loop
inputs [
 InteractLoop {type=WaveAsset}
 Interp Time {type=Time}
 Reset {type=Trigger}
 UE.Source.OnPlay {type=Trigger}
]
interfaces [UE.OutputFormat.Stereo:1.0, UE.Source:1.1]
outputs [
 "UE.OutputFormat.Stereo.Audio:0" {type=Audio}
 "UE.OutputFormat.Stereo.Audio:1" {type=Audio}
]
pages [Default {
  input_defaults {InteractLoop=&/PlayAsVillain_Interact_Searching_Loop_01, Interp Time=4, Reset=null, UE.Source.OnPlay=null}
  nodes [
   n1 = UE.InterpTo:Audio(
    Interp Time=^Interp Time
    Target=Value.Value:Float(Reset=^Reset, Set=^UE.Source.OnPlay) {also {defaults {Init Value=0.2, Target Value=1}}}.Output Value
   )
   WavePlayer_0 = UE.Wave Player:Stereo(Wave Asset=^InteractLoop, Play=^UE.Source.OnPlay) {also {defaults {Loop="<MetaSoundBuilderResult.SUCCEEDED: 0>"}, version [1,1]}}
   Multiply_0 = UE.Multiply:Audio by Float(AdditionalOperands=@n1.Value, PrimaryOperand=@WavePlayer_0.Out Left)
   Multiply_1 = UE.Multiply:Audio by Float(AdditionalOperands=@n1.Value, PrimaryOperand=@WavePlayer_0.Out Right)
  ]
  links [{dst="@UE.OutputFormat.Stereo.Audio:0", src=Multiply_0.Out}, {dst="@UE.OutputFormat.Stereo.Audio:1", src=Multiply_1.Out}]
 }]
properties {
 assetImportData=&self:AssetImportData
 attenuationSettings=/Game/Sounds/Attenuation/Fort_Item_Interact_Att
 bLooping=true
 duration=10000
 maxDistance=1850
 numChannels=2
 outputFormat=Stereo
 sampleRate=48000
 soundClassObject=&Fort_Gadgets_Items
 volume=0.7
}
