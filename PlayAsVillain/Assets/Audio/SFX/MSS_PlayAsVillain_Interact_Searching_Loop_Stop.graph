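# One-shot stereo MetaSound source that, going by its name, plays the stop sound for PlayAsVillain's searching interaction loop.
# On play, a single wave player plays one wave asset, exposed as an input, straight to both outputs and signals finished when the wave ends.
# Item-interact attenuation, a gadgets/items sound class and item concurrency control its range, mix and voice limiting.
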MetaSoundSource /PlayAsVillain/Assets/Audio/SFX/MSS_PlayAsVillain_Interact_Searching_Loop_Stop
inputs [
 UE.Source.OnPlay {type=Trigger}
 Wave Asset {type=WaveAsset}
]
interfaces [UE.OutputFormat.Stereo:1.0, UE.Source.OneShot:1.0, UE.Source:1.1]
outputs [
 "UE.OutputFormat.Stereo.Audio:0" {type=Audio}
 "UE.OutputFormat.Stereo.Audio:1" {type=Audio}
 UE.Source.OneShot.OnFinished {type=Trigger}
]
pages [Default {
  input_defaults {UE.Source.OnPlay=null, Wave Asset=&/PlayAsVillain_Interact_Searching_Loop_Stop_01}
  nodes [WavePlayer_0 = UE.Wave Player:Stereo(Play=^UE.Source.OnPlay, Wave Asset=^Wave Asset) {also {version [1,1]}}]
  links [
   {dst="@UE.Source.OneShot.OnFinished", src=WavePlayer_0.On Finished}
   {dst="@UE.OutputFormat.Stereo.Audio:0", src=WavePlayer_0.Out Left}
   {dst="@UE.OutputFormat.Stereo.Audio:1", src=WavePlayer_0.Out Right}
  ]
 }]
properties {
 assetImportData=&self:AssetImportData
 attenuationSettings=/Game/Sounds/Attenuation/Fort_Item_Interact_Att
 maxDistance=1850
 numChannels=2
 outputFormat=Stereo
 sampleRate=48000
 soundClassObject=&Fort_Gadgets_Items
 volume=0.6
 concurrencySet [&CS_Fort_Item]
}
