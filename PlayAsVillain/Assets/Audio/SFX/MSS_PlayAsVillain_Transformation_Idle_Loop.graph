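# Mono MetaSound source for the PlayAsVillain transformation item's idle loop sound effect.
# A single Wave Player starts on play, reads the Sound wave asset input and feeds the mono output, with the source marked as looping.
# It uses item attenuation, concurrency and gadget sound class settings, plus a control bus that modulates volume.
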MetaSoundSource /PlayAsVillain/Assets/Audio/SFX/MSS_PlayAsVillain_Transformation_Idle_Loop
inputs [
 Sound {type=WaveAsset}
 UE.Source.OnPlay {type=Trigger}
]
interfaces [UE.OutputFormat.Mono:1.0, UE.Source:1.1]
outputs ["UE.OutputFormat.Mono.Audio:0" {type=Audio}]
pages [Default {
  input_defaults {Sound=&/PlayAsVillain_Transformation_Idle_Loop_01, UE.Source.OnPlay=null}
  nodes [WavePlayer_0 = UE.Wave Player:Mono(Wave Asset=^Sound, Play=^UE.Source.OnPlay) {also {defaults {Loop="<MetaSoundBuilderResult.SUCCEEDED: 0>"}, version [1,1]}}]
  links [{dst="@UE.OutputFormat.Mono.Audio:0", src=WavePlayer_0.Out Mono}]
 }]
properties {
 assetImportData=&self:AssetImportData
 attenuationSettings=&/ATT_PlayAsVillain_Transformation_Item
 bLooping=true
 duration=10000
 maxDistance=3300
 sampleRate=48000
 soundClassObject=&Fort_Gadgets_Items
 volume=0.2
 concurrencySet [&CS_Fort_Item]
 modulationSettings {
  volumeRouting=Union
  volumeModulationDestination {bEnableModulation=true, modulators [&/CB_PlayAsVillain_Item]}
 }
}
