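# A one-shot 5.1 MetaSound source for the first-person sound of the villain transformation return in PlayAsVillain.
# On play, a single 5.1 wave player plays the exposed Wave Asset input, defaulting to a close return recording, and routes each channel straight to its matching output.
# Playback completion fires the one-shot finished trigger.
# The source uses a close transformation attenuation and the gadgets sound class.
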MetaSoundSource /PlayAsVillain/Assets/Audio/SFX/MSS_PlayAsVillain_Transformation_Return_1P
inputs [
 UE.Source.OnPlay {type=Trigger}
 Wave Asset {type=WaveAsset}
]
interfaces [UE.OutputFormat.5dot1:1.0, UE.Source.OneShot:1.0, UE.Source:1.1]
outputs [
 "UE.OutputFormat.5dot1.Audio:0" {type=Audio}
 "UE.OutputFormat.5dot1.Audio:1" {type=Audio}
 "UE.OutputFormat.5dot1.Audio:2" {type=Audio}
 "UE.OutputFormat.5dot1.Audio:3" {type=Audio}
 "UE.OutputFormat.5dot1.Audio:4" {type=Audio}
 "UE.OutputFormat.5dot1.Audio:5" {type=Audio}
 UE.Source.OneShot.OnFinished {type=Trigger}
]
pages [Default {
  input_defaults {UE.Source.OnPlay=null, Wave Asset=&/PlayAsVillain_Transformation_Return_1P_Close_01}
  nodes [WavePlayer_0 = UE.Wave Player:5dot1(Play=^UE.Source.OnPlay, Wave Asset=^Wave Asset) {also {version [1,1]}}]
  links [
   {dst="@UE.Source.OneShot.OnFinished", src=WavePlayer_0.On Finished}
   {dst="@UE.OutputFormat.5dot1.Audio:2", src=WavePlayer_0.Out Front Center}
   {dst="@UE.OutputFormat.5dot1.Audio:0", src=WavePlayer_0.Out Front Left}
   {dst="@UE.OutputFormat.5dot1.Audio:1", src=WavePlayer_0.Out Front Right}
   {dst="@UE.OutputFormat.5dot1.Audio:3", src=WavePlayer_0.Out Low Frequency}
   {dst="@UE.OutputFormat.5dot1.Audio:4", src=WavePlayer_0.Out Side Left}
   {dst="@UE.OutputFormat.5dot1.Audio:5", src=WavePlayer_0.Out Side Right}
  ]
 }]
properties {
 assetImportData=&self:AssetImportData
 attenuationSettings=&/ATT_PlayAsVillain_Transform_Close
 maxDistance=6800
 numChannels=6
 outputFormat=FiveDotOne
 sampleRate=48000
 soundClassObject=&Fort_Gadgets
}
