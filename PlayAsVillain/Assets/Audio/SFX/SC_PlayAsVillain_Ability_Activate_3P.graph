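# Sound cue for the third-person Play As Villain ability activation, layering close and distant recordings in the gadgets and items sound class.
# A mixer blends two branches, each with its own attenuation node, so the two layers follow separate distance curves.
# Random modulation varies pitch on the close layer and volume on the distant one.
# The cue overrides concurrency to cap simultaneous instances at a small count.
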SoundCue /PlayAsVillain/Assets/Audio/SFX/SC_PlayAsVillain_Ability_Activate_3P
root=Mixer_2
nodes [
 {
  id=Attenuation_0
  children [Modulator_13]
  properties {attenuationSettings=&/ATT_PlayAsVillain_Transform_Distant}
 }
 {id=Attenuation_1, children [Modulator_0], properties {attenuationSettings=&/ATT_PlayAsVillain_Transform_Close}}
 {id=Mixer_2, children [Attenuation_1, Attenuation_0], properties {inputVolume [1.1,1]}}
 {id=Modulator_0, children [WavePlayer_1], properties {pitchMax=1.1, pitchMin=0.9}}
 {id=Modulator_13, children [WavePlayer_2], properties {volumeMax=1.1, volumeMin=1}}
 {id=WavePlayer_1, properties {soundWaveAssetPtr=&/PlayAsVillain_Ability_Activate_3P_Close_01}}
 {id=WavePlayer_2, properties {soundWaveAssetPtr=&/PlayAsVillain_Ability_Activate_3P_Distant_01}}
]
properties {
 bOverrideConcurrency=true
 maxDistance=12500
 soundClassObject=&Fort_Gadgets_Items
 volumeMultiplier=0.7
 attenuationOverrides {distanceAlgorithm=NaturalSound, falloffDistance=3800, lPFRadiusMax=3000, lPFRadiusMin=2000, attenuationShapeExtents {x=2200}}
 concurrencyOverrides {maxCount=2}
}
