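# Sound cue for equipping an ability in PlayAsVillain.
# A Random node picks one of three equip wave variations with equal weight.
# The cue plays at reduced volume through the Fortnite item attenuation and the gadgets/items sound class, with a limited max audible distance.
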SoundCue /PlayAsVillain/Assets/Audio/SFX/SC_PlayAsVillain_Ability_Equip
root=Random_0
nodes [
 {id=Random_0, children [WavePlayer_0,WavePlayer_1,WavePlayer_2], properties {weights [1,1,1]}}
 {id=WavePlayer_0, properties {soundWaveAssetPtr=&/PlayAsVillain_Ability_Equip_01}}
 {id=WavePlayer_1, properties {soundWaveAssetPtr=&/PlayAsVillain_Ability_Equip_02}}
 {id=WavePlayer_2, properties {soundWaveAssetPtr=&/PlayAsVillain_Ability_Equip_03}}
]
properties {
 attenuationSettings=/FortMetasound/Items/AS_Fort_Item
 maxDistance=3500
 soundClassObject=&Fort_Gadgets_Items
 volumeMultiplier=0.15
}
