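# Sound cue for the Play As Villain ability expiring, routed to the gadgets and items sound class with item attenuation and concurrency.
# A quality-level switch sends its first slot through a modulator to an equally weighted random pick of three expire waves.
# The other three slots share a second modulator feeding a single Expire_01 wave, so random variation exists only in the first slot.
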SoundCue /PlayAsVillain/Assets/Audio/SFX/SC_PlayAsVillain_Ability_Expire
root=QualityLevel_0
nodes [
 {id=Modulator_0, children [WavePlayer_3]}
 {id=Modulator_1, children [Random_0]}
 {id=QualityLevel_0, children [Modulator_1,Modulator_0,Modulator_0,Modulator_0]}
 {id=Random_0, children [WavePlayer_0,WavePlayer_1,WavePlayer_2], properties {weights [1,1,1]}}
 {id=WavePlayer_0, properties {soundWaveAssetPtr=&/PlayAsVillain_Ability_Expire_01}}
 {id=WavePlayer_1, properties {soundWaveAssetPtr=&/PlayAsVillain_Ability_Expire_02}}
 {id=WavePlayer_2, properties {soundWaveAssetPtr=&/PlayAsVillain_Ability_Expire_03}}
 {id=WavePlayer_3, properties {soundWaveAssetPtr=&PlayAsVillain_Ability_Expire_01}}
]
properties {
 attenuationSettings=/FortMetasound/Items/AS_Fort_Item
 maxDistance=3500
 soundClassObject=&Fort_Gadgets_Items
 volumeMultiplier=0.45
 concurrencySet [&CS_Fort_Item]
}
