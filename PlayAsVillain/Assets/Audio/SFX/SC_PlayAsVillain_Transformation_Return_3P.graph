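# Sound cue for the third-person sound of the Play As Villain return transformation.
# A mixer layers a close recording and a distant recording, each with its own close or distant attenuation asset.
# Each layer passes through a modulator that slightly randomizes pitch.
# The cue uses the gadgets items sound class and overrides concurrency to cap simultaneous instances.
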SoundCue /PlayAsVillain/Assets/Audio/SFX/SC_PlayAsVillain_Transformation_Return_3P
root=Mixer_0
nodes [
 {
  id=Attenuation_0
  children [Modulator_1]
  properties {attenuationSettings=&/ATT_PlayAsVillain_Transform_Distant}
 }
 {id=Attenuation_1, children [Modulator_0], properties {attenuationSettings=&/ATT_PlayAsVillain_Transform_Close}}
 {id=Mixer_0, children [Attenuation_1, Attenuation_0], properties {inputVolume [1,1]}}
 {id=Modulator_0, children [WavePlayer_0], properties {pitchMax=1.1, pitchMin=0.9}}
 {id=Modulator_1, children [WavePlayer_3], properties {pitchMax=1.1, pitchMin=0.9}}
 {id=WavePlayer_0, properties {soundWaveAssetPtr=&/PlayAsVillain_Transformation_Return_3P_Close_01}}
 {id=WavePlayer_3, properties {soundWaveAssetPtr=&/PlayAsVillain_Transformation_Return_3P_Distant_01}}
]
properties {
 bOverrideConcurrency=true
 maxDistance=12500
 soundClassObject=&Fort_Gadgets_Items
 concurrencyOverrides {maxCount=2}
}
